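# Niagara idle effect for the BoneMarrow pickaxe, with a single SheenMesh emitter that renders short-lived hot-pink mesh particles.
# Spawning is distance-based via SpawnPerUnit, gated by a System.SpawnSheen probability, and the system exposes a User.bisSwinging bool.
# Curves over particle life animate dynamic material parameters and an alpha that dips out and returns.
# A second alpha scale indexed by System.Age dims the sheen as the system ages.
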NiagaraSystem /BRCosmetics/Weapons/FORT_Melee/Pickaxe_BoneMarrow/FX/NS_Pickaxe_BoneMarrow_Idle
dependencies {
 usedDataInterfaces [&NiagaraDataInterfaceCurve]
 usedDynamicInputs [&RandomRangeFloat, &MakeCustomFloatFromBool]
 usedModules [
  &self:SystemSpawnScript.SystemScriptSource.SystemScriptGraph.NiagaraNodeAssignment_0.SetVariables_#1
  &SystemState
  &self:SystemSpawnScript.SystemScriptSource.SystemScriptGraph.NiagaraNodeAssignment_3.SetVariables_#2
 ]
}
emitters {SheenMesh {
  emitterUpdateScript [
   &EmitterState()
   &SpawnPerUnit(
    Module.Use Max Movement Threshold=true
    Module.Use Movement Tolerance=false
    Spawn Probability:NiagaraFloat=$System.SpawnSheen
    Use Spawn Probability=true
   )
  ]
  particleSpawnScript [
   /Niagara/Modules/Spawn/Initialization/InitializeParticle(
    Color=<rgba 1 0.03 0.7252 1>
    Lifetime=1.5
    Module.Write Color=true
    Module.Write Lifetime=true
    Module.Write Mass=true
    Module.Write Material Random=true
    Module.Write Position=false
    Module.Write Scale=true
    Module.Write SpriteRotation=false
    Module.Write SpriteSize=true
    Position=<0 -1 -2>
    Sprite Size=<1 2>
   )
   /NiagaraFN/Sprite/FN_SpriteAlignmentAndFacing(SpriteAlignment:Vector3f=<0 0 1>, SpriteFacing:Vector3f=<1 0 0>)
  ]
  particleUpdateScript [
   &ParticleState()
   &DynamicMaterialParameters(
    Index 0 Param 1=&FloatFromCurve(FloatCurve=DataInterface() {propertyValues {Curve {keys [{arriveTangent=2.513, interpMode=RCIM_Cubic, time=0, value=0.2}, <at 0.3 1.8>]}}})
    Index 0 Param 2=1
    Index 0 Param 3=&FloatFromCurve(
     FloatCurve=DataInterface() {propertyValues {Curve {keys [{interpMode=RCIM_Cubic, tangentMode=RCTM_User, time=0.2, value=0}, <at 0.4 2 RCIM_Cubic>, <at 1 0 RCIM_Cubic>]}}}
     Scale Curve=0.2
    )
    Index 0 Param 4=&FloatFromCurve(FloatCurve=DataInterface() {propertyValues {Curve {keys [
         {interpMode=RCIM_Cubic, tangentMode=RCTM_User, time=0, value=0}
         {arriveTangent=1.2104, interpMode=RCIM_Cubic, leaveTangent=1.2104, tangentMode=RCTM_User, time=0.1, value=0.05}
         {arriveTangent=1.6138, interpMode=RCIM_Cubic, leaveTangent=1.6138, tangentMode=RCTM_User, time=0.3, value=0.4}
         <at 1 1>
        ]}}})
    Module.Param0WriteEnabled=true
    Module.Param1WriteEnabled=true
    Module.Param2WriteEnabled=true
    Module.Param3WriteEnabled=true
   )
   &CameraOffset(Camera Offset Amount=1)
   /SetVariables(Particles.CameraOffsetFlare=2) {isSetParameters=true}
   &SolveForcesAndVelocity(Clamp Velocity=false, Limit Acceleration=false)
   &ScaleColor(
    Scale Alpha=&FloatFromCurve(FloatCurve=DataInterface() {propertyValues {Curve {keys [
         {interpMode=RCIM_Cubic, tangentMode=RCTM_User, time=0, value=2}
         {arriveTangent=-4.695, interpMode=RCIM_Cubic, leaveTangent=-4.695, tangentMode=RCTM_User, time=0.2, value=1}
         {arriveTangent=-3.3333, interpMode=RCIM_Cubic, leaveTangent=-3.3333, tangentMode=RCTM_User, time=0.3, value=0}
         {arriveTangent=2.5, interpMode=RCIM_Cubic, leaveTangent=2.5, tangentMode=RCTM_User, time=0.5, value=0}
         {arriveTangent=2, interpMode=RCIM_Cubic, leaveTangent=2, time=0.6, value=1}
         <at 1 1 RCIM_Cubic>
        ]}}})
    ScaleA=true
    ScaleRGB=true
   )
   &ScaleColor(
    Scale Alpha=&FloatFromCurve(
     FloatCurve=DataInterface() {propertyValues {Curve {keys [
         {tangentMode=RCTM_User, time=0, value=1}
         {arriveTangent=0.0496, interpMode=RCIM_Cubic, leaveTangent=0.0496, tangentMode=RCTM_User, time=0.2, value=1}
         {arriveTangent=-2.3841, interpMode=RCIM_Cubic, leaveTangent=-2.3841, tangentMode=RCTM_User, time=0.3, value=0.025}
        ]}}}
     CurveIndex=$System.Age
     Scale Curve=22
    )
    ScaleA=true
    ScaleRGB=false
   ) {module=ScaleColor001}
  ]
  rendererClasses [&NiagaraMeshRendererProperties]
  renderers [{rendererClass=&NiagaraMeshRendererProperties, rendererIndex=0}]
 }}
summary {
 systemName=NS_Pickaxe_BoneMarrow_Idle
 emitters [{emitterName=SheenMesh, rendererClasses [&NiagaraMeshRendererProperties]}]
 userVariables [User.bisSwinging {type=NiagaraBool}]
}
